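# Sound cue for the firetruck horn blast in the Fort vehicle sounds.
# A single wave player feeds the horn sample through a modulator at the root, with the cue volume reduced.
# It plays in the impacts sound class and borrows a demolition-hit attenuation asset for medium-sized objects.
# A concurrency override blocks new instances instead of cutting off playing ones once the limit is reached.
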SoundCue /Game/Sounds/Fort_Vehicles/Fire/Firetruck_HornBlast_Cue
root=Modulator_0
nodes [
 {id=Modulator_0, children [WavePlayer_0]}
 {id=WavePlayer_0, properties {soundWaveAssetPtr=&/Firetruck_HornBlast}}
]
properties {
 attenuationSettings=/Game/Sounds/Attenuation/Fort_Demolition_Hit_Medium_Sized_Object_Att
 bOverrideConcurrency=true
 maxDistance=4000
 soundClassObject=&Fort_Impacts
 volumeMultiplier=0.6
 concurrencyOverrides {resolutionRule=PreventNew}
}
